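# Physics asset for the IntenseCello pickaxe cosmetic, defining a short jointed chain on its dyn_pizza bones.
# Capsule bodies cover a small master bone and three segments, which shorten from 01 to 03.
# Constraints link 01 to the master, 02 to 01 and 03 to 02, each allowing narrow, equal swing limits with a zero twist limit.
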PhysicsAsset /BRCosmetics/Weapons/FORT_Melee/Pickaxe_IntenseCello/Meshes/IntenseCello_Axe_Physics
bodies {
 dyn_pizza_01 [{shapeType=Capsule, length=10.8, radius=2.9, shapeName=dyn_pizza_01_capsule, center=<5 0 0>, rotation=<rot -90 0 0>}]
 dyn_pizza_02 [{shapeType=Capsule, length=7.5, radius=2.7, shapeName=dyn_pizza_02_capsule, center=<5 0 0>, rotation=<rot -90 0 0>}]
 dyn_pizza_03 [{shapeType=Capsule, length=4.4, radius=2.7, shapeName=dyn_pizza_03_capsule, center=<5 0 0>, rotation=<rot -90 0 0>}]
 dyn_pizza_MSTR [{shapeType=Capsule, length=0.5, radius=0.5, shapeName=dyn_pizza_MSTR_capsule, rotation=<rot 0 -90 0>}]
}
constraints [
 {bone1Name=dyn_pizza_01, bone2Name=dyn_pizza_MSTR, swing1LimitDegrees=15, swing2LimitDegrees=15, twistLimitDegrees=0}
 {bone1Name=dyn_pizza_02, bone2Name=dyn_pizza_01, swing1LimitDegrees=15, swing2LimitDegrees=15, twistLimitDegrees=0}
 {bone1Name=dyn_pizza_03, bone2Name=dyn_pizza_02, swing1LimitDegrees=15, swing2LimitDegrees=15, twistLimitDegrees=0}
]
